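# Conversation database for the DustyBun NPC, a theatrical mothman, giving its greeting and routing to further dialogue.
# The entry point leads to one Speech node with a default line plus contextual lines gated on first-ever conversation, per-match conversation count, or wearing the Moth outfit.
# First-meeting and outfit lines take priority from named curve rows.
# Speech then branches through link nodes to quests, basic services, a Season37 ArcanaAgate voice fork and DustyBun dialogue.
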ConversationDatabase /NPCLibrary/NPCs/DustyBun/NPCLibrary_Conversation_DustyBun
entry_points [{tag=Conversation.Entry.DustyBun, destinations [#1]}]
nodes [
 ConversationEntryPointNode_7 {
  id=#1
  class=ConversationEntryPointNode
  outputs [#2]
  properties {entryTag {tagName=Conversation.Entry.DustyBun}}
 }
 Speech_C_4 {
  id=#2
  class=Speech_C
  outputs [#3, #4, #5, #6]
  properties {
   nodeName=""
   generalConfig {
    defaultMessage="I am an omen of your darkest fears. The pay... is adequate!"
    contextualMessages [
     {
      message="You gaze upon the mothman! ILL TIDINGS ABOUND!!"
      contextRequirements [{requirement {class=FortControllerRequirement_IsFirstEverConversationWithNPC}}]
      priority {curve {rowName=Tandem.Speech.Priority.FirstEverInteraction}}
     }
     {
      message=Perchance I might ask for directions to the nearest bridge?
      contextRequirements [{requirement {
         class=FortControllerRequirement_CheckNPCMatchConversationCount
         properties {nPCUniqueIdTag {tagName=AISpawnerData.Type.Tandem.DustyBun}, requiredConversationCountConfig {requirement=GreaterThan}}
        }}]
      priority {curve {curveTable=null, rowName=None}}
     }
     {
      message="Quiver in fear! Await your fate! Bring me a snack!"
      contextRequirements [{requirement {
         class=FortControllerRequirement_CheckNPCMatchConversationCount
         properties {nPCUniqueIdTag {tagName=AISpawnerData.Type.Tandem.DustyBun}, requiredConversationCountConfig {requirement=GreaterThan}}
        }}]
      priority {curve {curveTable=null, rowName=None}}
     }
     {
      message="I am an omen of your darkest fears. The pay... is adequate!"
      contextRequirements [{requirement {
         class=FortControllerRequirement_CheckNPCMatchConversationCount
         properties {nPCUniqueIdTag {tagName=AISpawnerData.Type.Tandem.DustyBun}, requiredConversationCountConfig {requirement=GreaterThan}}
        }}]
      priority {curve {curveTable=null, rowName=None}}
     }
     {
      message="You are a pretender, sheathed in goldenrod lies. YOU DISGUST ME!"
      contextRequirements [{requirement {
         class=FortControllerRequirement_HasCID
         properties {allowedCIds [/BRCosmetics/Athena/Items/Cosmetics/Characters/CID_277_Athena_Commando_M_Moth]}
        }}]
      priority {curve {rowName=Tandem.Speech.Priority.OutfitReaction}}
     }
     {
      message="You are a pretender, sheathed in goldenrod lies. YOU DISGUST ME!"
      contextRequirements [{requirement {
         class=FortControllerRequirement_HasCID
         properties {allowedCIds [&CID_277_Athena_Commando_M_Moth]}
        }}]
      priority {curve {rowName=Tandem.Speech.Priority.OutfitReaction}}
     }
    ]
   }
  }
 }
 ConversationLinkNode_3 {id=#6, class=ConversationLinkNode, properties {remoteEntryTag {tagName=Conversation.Entry.DustyBun.Dialogue}}}
 ConversationLinkNode_2 {
  id=#5
  class=ConversationLinkNode
  properties {remoteEntryTag {tagName=Conversation.Entry.Season37.VoiceFork.ArcanaAgate.DustyBun}}
 }
 ConversationLinkNode_0 {id=#4, class=ConversationLinkNode, properties {remoteEntryTag {tagName=Conversation.Entry.BasicServices}}}
 ConversationLinkNode_1 {id=#3, class=ConversationLinkNode, properties {remoteEntryTag {tagName=Conversation.Entry.NPCLibrary.Quests}}}
]
