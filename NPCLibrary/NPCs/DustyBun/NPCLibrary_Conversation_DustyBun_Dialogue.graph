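# Conversation database for the DustyBun NPC dialogue, where a theatrical omen-bringer, referred to as the Mothman, answers the player's questions.
# The entry tag leads to an "Ask about..." speech hub branching into three topics (reason for visiting, being mysterious, daily schedule) and a Back node.
# Each topic is a linear chain of four Speech nodes ending in a ReturnToStart node labeled BACK.
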ConversationDatabase /NPCLibrary/NPCs/DustyBun/NPCLibrary_Conversation_DustyBun_Dialogue
entry_points [{tag=Conversation.Entry.DustyBun.Dialogue, destinations [#1]}]
nodes [
 ConversationEntryPointNode_0 {
  id=#1
  class=ConversationEntryPointNode
  outputs [#2]
  properties {entryTag {tagName=Conversation.Entry.DustyBun.Dialogue}}
 }
 Speech_C_0 {
  id=#2
  class=Speech_C
  outputs [#3, #4, #5, #6]
  properties {nodeName="", softTaskIcon=&T_UI_NPC_question, generalConfig {defaultMessage="Ask your questions, mortal!"}}
  sub_nodes [ConversationChoiceNode_5 {class=ConversationChoiceNode, properties {defaultChoiceDisplayText=Ask about...}}]
 }
 Speech_C_8 {
  id=#6
  class=Speech_C
  outputs [#7]
  properties {
   nodeName=""
   softTaskIcon=&T_UI_NPC_question
   generalConfig {defaultMessage=I like to start my day off by skulking near bridges.}
  }
  sub_nodes [ConversationChoiceNode_15 {class=ConversationChoiceNode, properties {defaultChoiceDisplayText=The Mothman's schedule}}]
 }
 Speech_C_9 {
  id=#7
  class=Speech_C
  outputs [#8]
  properties {
   nodeName=""
   softTaskIcon=&T_UI_NPC_question
   generalConfig {defaultMessage=Then I spend a few hours hissing ill tidings from the shadows.}
  }
 }
 Speech_C_10 {
  id=#8
  class=Speech_C
  outputs [#9]
  properties {
   nodeName=""
   softTaskIcon=&T_UI_NPC_question
   generalConfig {defaultMessage="Then, catered lunch. I get a cobb salad."}
  }
 }
 Speech_C_11 {
  id=#9
  class=Speech_C
  outputs [#10]
  properties {
   nodeName=""
   softTaskIcon=&T_UI_NPC_question
   generalConfig {defaultMessage="And then at that point, I usually spend the rest of the day watching my dark prophecies unfold. With popcorn. Mmm."}
  }
 }
 ReturnToStart_C_1 {
  id=#10
  class=ReturnToStart_C
  properties {softTaskIcon=&T_UI_NPC_Back}
  sub_nodes [ConversationChoiceNode_32 {class=ConversationChoiceNode, properties {defaultChoiceDisplayText=BACK}}]
 }
 Speech_C_3 {
  id=#5
  class=Speech_C
  outputs [#11]
  properties {
   nodeName=""
   softTaskIcon=&T_UI_NPC_question
   generalConfig {defaultMessage=Embodying the living energy of Mystery is... exhausting.}
  }
  sub_nodes [ConversationChoiceNode_2 {class=ConversationChoiceNode, properties {defaultChoiceDisplayText=Being Mysterious}}]
 }
 Speech_C_4 {
  id=#11
  class=Speech_C
  outputs [#12]
  properties {
   nodeName=""
   softTaskIcon=&T_UI_NPC_question
   generalConfig {defaultMessage="Sometimes I wonder what it might be like to *solve* a mystery, instead of *being* a mystery."}
  }
 }
 Speech_C_5 {
  id=#12
  class=Speech_C
  outputs [#13]
  properties {
   nodeName=""
   softTaskIcon=&T_UI_NPC_question
   generalConfig {defaultMessage="...ALAS! I almost launched into a pretty personal monologue there."}
  }
 }
 Speech_C_6 {
  id=#13
  class=Speech_C
  outputs [#14]
  properties {
   nodeName=""
   softTaskIcon=&T_UI_NPC_question
   generalConfig {defaultMessage="BEGONE, MORTAL! OUT OF MY HEAD! I have HAUNTINGS to HAUNT!"}
  }
 }
 ReturnToStart_C_0 {
  id=#14
  class=ReturnToStart_C
  properties {softTaskIcon=&T_UI_NPC_Back}
  sub_nodes [ConversationChoiceNode_3 {class=ConversationChoiceNode, properties {defaultChoiceDisplayText=BACK}}]
 }
 Back_C_0 {
  id=#4
  class=Back_C
  sub_nodes [ConversationChoiceNode_7 {class=ConversationChoiceNode, properties {defaultChoiceDisplayText=BACK}}]
 }
 Speech_C_7 {
  id=#3
  class=Speech_C
  outputs [#15]
  properties {
   nodeName=""
   softTaskIcon=&T_UI_NPC_question
   generalConfig {defaultMessage="A miasma of misfortune clouds this Island. "}
  }
  sub_nodes [ConversationChoiceNode_0 {class=ConversationChoiceNode, properties {defaultChoiceDisplayText=Why are you here?}}]
 }
 Speech_C_15 {
  id=#15
  class=Speech_C
  outputs [#16]
  properties {
   nodeName=""
   softTaskIcon=&T_UI_NPC_question
   generalConfig {defaultMessage="I am both omen-bringer and omen. I am attracted to ill-tidings, like a moth to a... bad feeling."}
  }
 }
 Speech_C_1 {
  id=#16
  class=Speech_C
  outputs [#17]
  properties {
   nodeName=""
   softTaskIcon=&T_UI_NPC_question
   generalConfig {defaultMessage="And so I am here! Due to the aforementioned miasma! OF MISFORTUNE!"}
  }
 }
 Speech_C_2 {
  id=#17
  class=Speech_C
  outputs [#18]
  properties {
   nodeName=""
   softTaskIcon=&T_UI_NPC_question
   generalConfig {defaultMessage="My advice: Hold on to your posteriors, mortals! I don't make housecalls for nothin'!"}
  }
 }
 ReturnToStart_C_23 {
  id=#18
  class=ReturnToStart_C
  properties {softTaskIcon=&T_UI_NPC_Back}
  sub_nodes [ConversationChoiceNode_1 {class=ConversationChoiceNode, properties {defaultChoiceDisplayText=BACK}}]
 }
]
